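# Additive, unlit, two-sided spline material that draws a rainbow which shifts and fades with the time of day.
# A remapped TimeOfDay_0-1 drives a sine window that fades opacity and collapses SplineThicken width to zero outside it.
# The same time value lerps the world offset from StartOffset to EndOffset.
# A gradient texture supplies color with U mirrored by view-facing side, while sphere masks on the UVs soften opacity.
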Material /Figment_S07/Effects/Fort_Effects/Materials/M_Rainbow_Spline
graph {
 properties {blendMode=BLEND_Additive, pixelDepthOffsetMode=PDOM_Legacy, shadingModel=MSM_Unlit, twoSided=true, usedShadingModels=Unlit}
 nodes [
  n1 = Clamp(None=LinearInterpolate(Alpha=CollectionParameter() {collection=&FortniteMaterialParameters, parameterName=TimeOfDay_0-1}) {constA=-0.5, constB=1.7})
  n2 = Clamp(None=Sine(None=@n1) {period=2})
  n3 = Multiply(A=ScalarParameter() {defaultValue=200000, parameterName=Width}, B=Clamp(None=Ceil(None=@n2)))
  n4 = VectorParameter() {parameterName=StartOffset, defaultValue=<rgba 0 -400 100 0>}
  n5 = VectorParameter() {parameterName=EndOffset, defaultValue=<rgba 600 1000 -500 0>}
  Add(
   A=MaterialFunctionCall(WidthBase=@n3, WidthTip=@n3) {materialFunction=&SplineThicken}.WorldPositionOffset
   B=Multiply(A=LinearInterpolate(A=@n4.RGB, B=@n5.RGB, Alpha=@n1)) {constB=1000}
  )
  n6 = TextureCoordinate()
  n7 = Constant() {r=0.5}
  LinearInterpolate(Alpha=SphereMask(A=ComponentMask(None=@n6) {g=true}, B=@n7, Radius=@n7) {hardnessPercent=0}) {constA=0.25}
  MaterialFunctionCall(Speed=Constant() {r=0.1}) {materialFunction=&DebugTimeSine}
  n8 = TextureCoordinate()
  n9 = MaterialFunctionCall(Float2=@n8) {materialFunction=&BreakOutFloat2Components}
  n10 = DotProduct(A=VertexNormalWS(), B=CameraVectorWS())
  n11 = LinearInterpolate(
   A=@n8
   B=AppendVector(A=OneMinus(None=@n9.R), B=@n9.G)
   Alpha=MaterialFunctionCall!(Mask Offset Value, Number Before White Result, Gradient=Clamp(None=@n10)) {materialFunction=/Engine/Functions/Engine_MaterialFunctions02/ValueStep}.Result
  )
  n12 = ScalarParameter() {defaultValue=0.38, parameterName=ColorScroll}
  Multiply(
   A=ScalarParameter() {defaultValue=1.4, parameterName=Brightness}
   B=TextureSample(UVs=Add(A=@n11, B=@n12)) {texture=/Game/Global/Textures/Default/Masks_and_Gradients/T_Grad_Rainbow}.RGB
  )
  n13 = TextureCoordinate()
  n14 = Constant() {r=0.5}
  n15 = ComponentMask(None=TextureCoordinate()) {g=true}
  n16 = Constant() {r=0.5}
  n17 = Multiply(
   A=SphereMask(A=ComponentMask(None=@n13) {r=true}, B=@n14, Radius=@n14) {hardnessPercent=0}
   B=LinearInterpolate(B=LinearInterpolate(Alpha=@n1) {constA=1, constB=0.1}, Alpha=SphereMask(A=@n15, B=@n16, Radius=@n16) {hardnessPercent=0}) {constA=1.2, constB=0.5}
  )
  n18 = ScalarParameter() {defaultValue=2, parameterName=OpacityPower}
  Multiply(A=Multiply(A=Power(Base=@n17, Exp=@n18), B=ScalarParameter() {defaultValue=0.35, parameterName=OpacityMult}), B=@n2)
 ]
}
